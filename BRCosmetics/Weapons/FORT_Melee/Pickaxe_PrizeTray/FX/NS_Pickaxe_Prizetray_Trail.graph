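# Niagara ribbon trail for the Prize Tray pickaxe, built on shared FN anim-trail modules for melee swings.
# One emitter spawns short-lived, locked particles between two bone or socket indices, drawn by three ribbon renderers.
# Particle updates derive separate screen-oriented, speedline and soft-trail attributes, with time-of-day tinting on the main and soft colors.
# System-age curves ramp width in and sharply cut main alpha, alongside per-particle fades.
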NiagaraSystem /BRCosmetics/Weapons/FORT_Melee/Pickaxe_PrizeTray/FX/NS_Pickaxe_Prizetray_Trail
dependencies {
 usedDataInterfaces [&NiagaraDataInterfaceCurve, &NiagaraDataInterfaceColorCurve, &NiagaraDataInterfaceSocketReader]
 usedDynamicInputs [&FNTimeOfDayScalar]
 usedModules [
  &self:SystemSpawnScript.SystemScriptSource.SystemScriptGraph.NiagaraNodeAssignment_0.SetVariables_#1
  &SystemState
  &self:SystemSpawnScript.SystemScriptSource.SystemScriptGraph.NiagaraNodeAssignment_1.SetVariables_#2
  &self:SystemSpawnScript.SystemScriptSource.SystemScriptGraph.NiagaraNodeAssignment_2.SetVariables_#3
 ]
}
emitters {Base {
  emitterSpawnScript [&FN_RibbonIDOffsetUtility()]
  emitterUpdateScript [
   &EmitterState()
   &FN_SpawnAnimTrail(Spawn Interval=0)
  ]
  particleSpawnScript [
   &FN_InitializeAnimTrailBoneIndecies(First Bone or Socket Index=1, Second Bone or Socket Index=0)
   &FN_InitializeAnimTrail(AnimTrailWidth=1.25, Lifetime=0.25)
  ]
  particleUpdateScript [
   &FN_UpdateAnimTrail(Lock Particles=true)
   &ScaleColor(
    Scale Alpha=&FloatFromCurve(
     FloatCurve=DataInterface() {propertyValues {Curve {
        defaultValue=1
        keys [{interpMode=RCIM_Cubic, tangentMode=RCTM_User, time=0.15, value=1}, {interpMode=RCIM_Cubic, tangentMode=RCTM_User, time=0.5, value=0}]
       }}}
     Scale Curve=&FloatFromCurve(FloatCurve=DataInterface() {propertyValues {Curve {keys [<at 0.2 1>,<at 0.23 0>]}}}, CurveIndex=$Engine.System.Age)
    )
    Scale RGB=&MakeVectorFromLinearColorRGB(Color=&LerpLinearColors(
      StartColor=&FNTimeOfDayColor(
       Night Color=<rgba 0.0441 0.2382 0.3229 1>
       Morning Color=<rgba 1 0.6018 0.7909 1>
       Day Color=<rgba 1 1 1 1>
       Evening Color=<rgba 0.7656 0.4426 0.4702 1>
      )
      EndColor=<rgba 1 1 1 1>
      LerpFactor=0.25
     ))
    ScaleA=true
    ScaleRGB=true
   ) {module=ScaleColor002}
   /SetVariables(
    Particles.ScreenOrientedTrailColor:LinearColor=&MultiplyLinearColor(Linear Color=$StackContext.Color, Scale Factor=<1 1 1 0.5>)
    Particles.ScreenOrientedTrailWidth:NiagaraFloat=&Multiply_Float(A=$StackContext.RibbonWidth, B=0.6)
   ) {isSetParameters=true}
   /SetVariables(Particles.SpeedlineColor:LinearColor=&MakeLinearColorFromVectorAndFloat(
     "Vector (RGB)"=<0 0 0>
     "Float (Alpha)"=&FloatFromCurve(FloatCurve=DataInterface() {propertyValues {Curve {keys [
          {interpMode=RCIM_Cubic, tangentMode=RCTM_User, time=0, value=1}
          {arriveTangent=-0.1438, interpMode=RCIM_Cubic, leaveTangent=-0.1438, tangentMode=RCTM_User, time=1, value=0}
         ]}}})
    )) {isSetParameters=true}
   /SetVariables(Particles.SoftTrailColor:LinearColor=&MakeLinearColorFromVectorAndFloat(
     "Vector (RGB)"=&Lerp_Vector(
      A=&MakeVectorFromLinearColorRGB(Color=&ColorFromCurve(ColorCurve=DataInterface() {propertyValues {
          AlphaCurve {keys [<at 0 1 RCIM_Cubic>,<at 1 0 RCIM_Cubic>]}
          BlueCurve {keys [<at 0 0.0661 RCIM_Cubic>,<at 1 0.0266 RCIM_Cubic>]}
          GreenCurve {keys [<at 0 0.2463 RCIM_Cubic>,<at 1 0.0792 RCIM_Cubic>]}
          RedCurve {keys [<at 0 0.5521 RCIM_Cubic>,<at 1 0.1667 RCIM_Cubic>]}
         }}))
      B=&MakeVectorFromLinearColorRGB(Color=&FNTimeOfDayColor(
        Night Color=<rgba 0.0115 0.0249 0.1302 1>
        Morning Color=<rgba 0.2656 0.119 0.1746 1>
        Day Color=<rgba 0.2344 0.2344 0.2344 1>
        Evening Color=<rgba 0.3073 0.1867 0.1136 1>
       ))
      Alpha=0.5
     )
     "Float (Alpha)"=&FloatFromCurve(FloatCurve=DataInterface() {propertyValues {Curve {keys [<at 0 1 RCIM_Cubic>,<at 0.75 0 RCIM_Cubic>]}}}, Scale Curve=0.5)
    )) {isSetParameters=true}
   &ScaleRibbonWidth(Ribbon Width Scale=&FloatFromCurve(
     FloatCurve=DataInterface() {propertyValues {Curve {
        defaultValue=1
        keys [
         {arriveTangent=-0.6303, interpMode=RCIM_Cubic, leaveTangent=-0.6303, tangentMode=RCTM_User, time=0, value=0}
         {arriveTangent=0.088, interpMode=RCIM_Cubic, leaveTangent=0.088, tangentMode=RCTM_User, time=0.1, value=0.9847}
        ]
       }}}
     CurveIndex=$Engine.System.Age
    ))
   /SetVariables(Particles.SoftTrailSize:NiagaraFloat=&Multiply_Float(A=$Particles.RibbonWidth, B=0.66)) {isSetParameters=true}
  ]
  rendererClasses [&NiagaraRibbonRendererProperties]
  renderers [
   {rendererClass=&NiagaraRibbonRendererProperties}
   {rendererClass=&NiagaraRibbonRendererProperties, rendererIndex=1}
   {rendererClass=&NiagaraRibbonRendererProperties, rendererIndex=2}
  ]
 }}
summary {
 systemName=NS_Pickaxe_Prizetray_Trail
 emitters [{emitterName=Base, rendererClasses [&NiagaraRibbonRendererProperties]}]
 userVariables [
  User.RibbonIDOffset {type=NiagaraInt32}
  User.SwingSectionIndex {type=NiagaraInt32}
 ]
}
